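# Niagara effect for the QueenTruth emote that draws a spinning skateboard motion blur and a delayed impact flash.
# Main_Ring spawns short-lived rotating meshes, using system age to shift randomized color from green to purple and raise their height.
# BlurRing adds fainter counter-rotating meshes; both rings tint by time of day, with BlurRing also computing a front-end-adjusted alpha.
# Impact_Flare_and_Corona bursts two sprite-and-mesh flares, partitioned to opposite sides below the origin.
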NiagaraSystem /BRCosmetics/Animation/Game/MainPlayer/Emotes/QueenTruth/FX/NS_QueenTruth_SkateboardBlur
dependencies {
 usedDataInterfaces [&NiagaraDataInterfaceCurve, &NiagaraDataInterfaceArrayFloat3]
 usedDynamicInputs []
 usedModules [&SystemState, &BlendSpaceHelper]
}
emitters {
 BlurRing {
  emitterUpdateScript [
   &EmitterState(Life Cycle Mode=::NewEnumerator1, Loop Behavior=::NewEnumerator1, Loop Duration=0.3, UseLoopDelay=false)
   &SpawnRate(SpawnRate=15, Use Spawn Probability=false)
  ]
  particleSpawnScript [
   &InitializeParticle(
    Color Maximum=<rgba 0.4 0.47 0.5 0.6>
    Color Minimum=<rgba 0.8 0.9367 1 0.7>
    Color Mode=::NewEnumerator2
    Lifetime=0.25
    Mesh Scale=&Multiply_VectorByFloat(Vector=<1 1.7651 1>, Float=0.1216)
    Mesh Scale Mode=::NewEnumerator4
    Position Offset=&RandomRangeVector(Minimum=<-1 -30 -1>, Maximum=<1 30 1>)
    UsePositionOffset=true
    Write Material Random=true
   )
   &CameraOffset(Camera Offset Amount=15)
   &InitialMeshOrientation(
    Mesh Orientation Mode=::NewEnumerator1
    Rotation=&RandomRangeVector(Minimum=<0 0 0>, Maximum=<0 1 0>)
    Use Rotation Vector=true
   )
  ]
  particleUpdateScript [
   &ParticleState()
   &ScaleMeshSize(Scale Factor=&Multiply_VectorByFloat(
     Vector=<1 1 1>
     Float=&Lerp_Float(
      A=0.9
      B=0.9
      Alpha=&Multiply_Float(
       A=&VectorLength(Vector=&ConvertPositionToVector(Input Position=$Particles.Position, Conversion Operation:ENiagara_LWCConvertPosToVec=::NewEnumerator0))
       B=0.0802
      )
     )
    )) {module=ScaleMeshSize001}
   &UpdateMeshOrientation(Rotation Rate=2, Rotation Vector=<0 1 0>) {module=UpdateMeshOrientation002}
   &DynamicMaterialParameters(
    Index 0 Param 1=&FloatFromCurve(FloatCurve=DataInterface() {propertyValues {Curve {keys [
         {arriveTangent=2.4535, interpMode=RCIM_Cubic, leaveTangent=2.4535, tangentMode=RCTM_User, time=0, value=0}
         {arriveTangent=1, interpMode=RCIM_Cubic, leaveTangent=1, time=0.0239, value=0.3818}
         {interpMode=RCIM_Cubic, tangentMode=RCTM_User, time=0.5, value=0.5}
         {arriveTangent=-2.5865, tangentMode=RCTM_User, time=1, value=0}
        ]}}})
    Module.Index0Param0Write=true
    Module.Index0Param1Write=false
    Module.Index0Param2Write=false
    Module.Index0Param3Write=false
   )
   &ScaleColor(
    Scale Alpha=&FloatFromCurve(FloatCurve=DataInterface() {propertyValues {Curve {keys [
         {arriveTangent=6.3511, interpMode=RCIM_Cubic, leaveTangent=6.3511, tangentMode=RCTM_User, time=0, value=0}
         {interpMode=RCIM_Cubic, tangentMode=RCTM_User, time=0.5, value=1}
         {arriveTangent=-6.5462, tangentMode=RCTM_User, time=1, value=0}
        ]}}})
    Scale RGB=<0.5 0.5 0.5>
    ScaleA=true
    ScaleRGB=true
   )
   /SetVariables(Particles.New Linear Color:LinearColor=&MakeLinearColorFromVectorAndFloat(
     "Vector (RGB)"=&MakeVectorFromLinearColorRGB(Color=&FNTimeOfDayColor(
       Night Color=<rgba 0.1253 0.1914 0.5 1>
       Morning Color=<rgba 0.5 0.3383 0.3719 1>
       Day Color=<rgba 0.5 0.5 0.5 1>
       Evening Color=<rgba 0.4474 0.4547 0.4578 1>
      ))
     "Float (Alpha)"=&Multiply_Float(
      A=&Multiply_Float(
       A=&FN_FrontEndFloatModifier(Gameplay=1, Front End=1.4)
       B=&FNTimeOfDayScalar(Night Scale=0.5, Morning Scale=0.3, Evening Scale=0.3)
      )
      B=&FloatFromCurve(FloatCurve=DataInterface() {propertyValues {Curve {keys [
           {arriveTangent=2, interpMode=RCIM_Cubic, time=0.0164, value=0.8082}
           {interpMode=RCIM_Cubic, tangentMode=RCTM_User, time=0.5, value=1}
           {tangentMode=RCTM_User, time=1, value=0}
          ]}}})
     )
    )) {isSetParameters=true}
   /SetVariables(Particles.BlurMeshScale:Vector3f=&Multiply_Vector(A=$Particles.Initial.Scale, B=<0.85 0.85 0.85>)) {isSetParameters=true}
  ]
  rendererClasses [&NiagaraMeshRendererProperties]
  renderers [{rendererClass=&NiagaraMeshRendererProperties, rendererIndex=0}]
 }
 Impact_Flare_and_Corona {
  emitterUpdateScript [
   &EmitterState()
   &SpawnBurst_Instantaneous(Spawn Count=2, Spawn Time=0.6, Use Spawn Probability=false)
  ]
  particleSpawnScript [
   &PartitionParticles()
   /Niagara/Modules/Spawn/Initialization/InitializeParticle(
    Color=<rgba 1 1 1 0.3>
    Lifetime=0.2
    Mesh Scale=&UniformRangedVector(Minimum=<0.1 1.6 1.6>, Maximum=<0.1 1.8 1.8>)
    Module.Write Color=true
    Module.Write Lifetime=true
    Module.Write Mass=true
    Module.Write Material Random=true
    Module.Write Position=true
    Module.Write Scale=true
    Module.Write SpriteRotation=false
    Module.Write SpriteSize=true
    Position=&SimulationPosition()
    Sprite Size=<133.3333 116.6667>
   ) {enabled=false}
   /Niagara/Modules/Spawn/Initialization/V2/InitializeParticle(
    Color=<rgba 1 1 1 0.3>
    Lifetime=0.2
    Mesh Scale Max=<0.12 1.8 1.8>
    Mesh Scale Min=<0.12 1.6 1.6>
    Mesh Scale Mode=::NewEnumerator2
    Position Offset=&MaskVector(Mask Source=1, Mask Parameter=$Particles.PartitionParticles.Partition, Mask=0, Masked Vector=<0 -30 -71>, Passthrough Vector=<0 30 -71>)
    UsePositionOffset=true
    Write Material Random=true
   ) {module=InitializeParticle001}
   /SetVariables(Particles.CoronaSpriteSize=&Vector2DFromFloat(Value=150)) {isSetParameters=true}
   /SetVariables(Particles.ImpactColor:LinearColor=&MakeLinearColorFromVectorAndFloat("Vector (RGB)"=<0.5 0.5 0.5>, "Float (Alpha)"=&Multiply_Float(A=1, B=1))) {isSetParameters=true}
   &InitialMeshOrientation(Orientation Vector=<0 0 1>, Use Orientation Vector=true, Use Rotation Vector=true)
   &ApplyInitialForces()
   &AlignSpriteToMeshOrientation(SetAlignment=true, SetFacing=true)
  ]
  particleUpdateScript [
   &ParticleState()
   &DynamicMaterialParameters(
    Index 0 Param 1=&FloatFromCurve(FloatCurve=DataInterface() {propertyValues {Curve {keys [<at 0 1 RCIM_Cubic>,<at 0.8 0 RCIM_Cubic>]}}})
    Index 0 Param 2=2
    Index 0 Param 3=1
    Index 0 Param 4=1
    Index 1 Param 1=&FloatFromCurve(FloatCurve=DataInterface() {propertyValues {Curve {keys [
         {arriveTangent=-2.3163, interpMode=RCIM_Cubic, time=0, value=2}
         {arriveTangent=-1.9019, interpMode=RCIM_Cubic, leaveTangent=-1.9019, tangentMode=RCTM_User, time=1, value=0.4}
        ]}}})
    Index 1 Param 2=0.4
    Index 1 Param 3=0.6
    Index 1 Param 4=1
    Index 2 Param 1=&FloatFromCurve(FloatCurve=DataInterface() {propertyValues {Curve {keys [<at 0.46 0>,<at 1 1>]}}})
    Index 2 Param 2=45
    Module.Param0WriteEnabled=true
    Module.Param0WriteEnabled1=true
    Module.Param0WriteEnabled2=true
    Module.Param1WriteEnabled=true
    Module.Param1WriteEnabled1=true
    Module.Param1WriteEnabled2=true
    Module.Param2WriteEnabled=true
    Module.Param2WriteEnabled1=true
    Module.Param2WriteEnabled2=true
    Module.Param3WriteEnabled=true
    Module.Param3WriteEnabled1=true
    Module.Param3WriteEnabled2=true
    Write Parameter Index 1=true
    Write Parameter Index 2=true
   )
   /SetVariables(Particles.CoronaSpriteColor:LinearColor=&MakeLinearColorFromVectorAndFloat(
     "Vector (RGB)"=&MakeVectorFromLinearColorRGB(Color=<rgba 0.5521 0.5521 0.5521 1>)
     "Float (Alpha)"=&Multiply_Float(
      A=&FloatFromCurve(FloatCurve=DataInterface() {propertyValues {Curve {keys [<at 0.35 1.5 RCIM_Cubic>,<at 0.95 0>]}}})
      B=3.9
     )
    )) {isSetParameters=true}
   /SetVariables(Particles.ImpactColor:LinearColor=<rgba 0.4 0.4 0.4 1>) {isSetParameters=true}
   &ScaleColor(
    Scale Alpha=&FloatFromCurve(FloatCurve=DataInterface() {propertyValues {Curve {keys [<at 0 1>,<at 1 0>]}}})
    Scale RGB=&MakeVectorFromLinearColorRGB(Color=<rgba 1 1 1 1>)
    ScaleA=true
    ScaleRGB=true
   )
   &ScaleMeshSize(Scale Factor=&VectorFromFloat(Value=&FloatFromCurve(FloatCurve=DataInterface() {propertyValues {Curve {keys [<at 0 0.294 RCIM_Cubic>,<at 1 1.05>]}}}, Scale Curve=1.4)))
  ]
  rendererClasses [&NiagaraSpriteRendererProperties, &NiagaraMeshRendererProperties]
  renderers [
   {rendererClass=&NiagaraSpriteRendererProperties}
   {rendererClass=&NiagaraSpriteRendererProperties, rendererIndex=1}
   {rendererClass=&NiagaraMeshRendererProperties, rendererIndex=2}
  ]
 }
 Main_Ring {
  emitterUpdateScript [
   &EmitterState(Life Cycle Mode=::NewEnumerator1, Loop Behavior=::NewEnumerator1, Loop Duration=0.3, UseLoopDelay=false)
   &SpawnRate(SpawnRate=60, Use Spawn Probability=false)
  ]
  particleSpawnScript [
   /SetVariables(
    Particles.Determinism:NiagaraFloat=&MakeCustomFloatFromBool(
     Bool=&SetBoolByFloatComparison(
      A=&RandomRangeFloat(Minimum=0, Maximum=1)
      B=&FloatFromCurve(FloatCurve=DataInterface() {propertyValues {Curve {keys [<at 0 0.1>,<at 0.3 1>]}}}, CurveIndex=$Engine.System.Age)
     )
     True Float=1
     False Float=0
    )
    Particles.SideDeterminism:NiagaraFloat=&RandomRangeFloat(Minimum=0, Maximum=1)
   ) {isSetParameters=true}
   &PartitionParticles()
   &InitializeParticle(
    Color=&LerpLinearColors(StartColor=<rgba 0.0404 0.548 0.1376 0.8>, EndColor=<rgba 0.4229 0.1829 0.542 0.8>, LerpFactor=$Particles.Determinism)
    Lifetime=&Lerp_Float(A=0.2, B=0.15, Alpha=$Particles.Determinism)
    Mesh Scale=&Multiply_VectorByFloat(
     Vector=&MakeVector(X=1, Y=&Lerp_Float(A=2, B=1, Alpha=$Particles.Determinism), Z=1)
     Float=&RandomRangeFloat(Minimum=0.16, Maximum=0.2)
    )
    Mesh Scale Mode=::NewEnumerator4
    Position Offset=&Lerp_Vector(
     A=&Lerp_Vector(A=<-1 -40 5>, B=<-1 40 5>, Alpha=$Particles.SideDeterminism)
     B=&SelectVectorFromArray(
      Array Sampling Mode=::NewEnumerator0
      Direct Array Index=?
      Vector Selection Array=DataInterface() {propertyValues {FloatData [<0 -27 0>, <1 27 1>]}}
      Random Seed=0
     )
     Alpha=$Particles.Determinism
    )
    UsePositionOffset=true
    Write Material Random=true
   )
   &CameraOffset(Camera Offset Amount=15)
   &InitialMeshOrientation(
    Mesh Orientation Mode=::NewEnumerator1
    Rotation=&RandomRangeVector(Minimum=<0 0 0>, Maximum=<0 1 0>)
    Use Rotation Vector=true
   )
  ]
  particleUpdateScript [
   &ParticleState()
   &ScaleMeshSize(Scale Factor=&Multiply_VectorByFloat(
     Vector=<1 1 1>
     Float=&Lerp_Float(
      A=1
      B=1
      Alpha=&Multiply_Float(
       A=&VectorLength(Vector=&ConvertPositionToVector(Input Position=$Particles.Position, Conversion Operation:ENiagara_LWCConvertPosToVec=::NewEnumerator0))
       B=0.0802
      )
     )
    )) {module=ScaleMeshSize001}
   &UpdateMeshOrientation(Rotation Rate=-1, Rotation Vector=<0 1 0>) {module=UpdateMeshOrientation002}
   &ScaleMeshSize(Scale Factor=&VectorFromFloat(Value=&FloatFromCurve(FloatCurve=DataInterface() {propertyValues {Curve {keys [<at 0 0.7>,<at 1 1>]}}})))
   &DynamicMaterialParameters(
    Index 0 Param 1=&FloatFromCurve(FloatCurve=DataInterface() {propertyValues {Curve {keys [
         {arriveTangent=2.4535, interpMode=RCIM_Cubic, leaveTangent=2.4535, tangentMode=RCTM_User, time=0, value=0}
         {interpMode=RCIM_Cubic, tangentMode=RCTM_User, time=0.5, value=0.5}
         {arriveTangent=-2.5865, tangentMode=RCTM_User, time=1, value=0}
        ]}}})
    Module.Index0Param0Write=true
    Module.Index0Param1Write=false
    Module.Index0Param2Write=false
    Module.Index0Param3Write=false
   )
   &ScaleColor(
    Scale Alpha=&FloatFromCurve(FloatCurve=DataInterface() {propertyValues {Curve {keys [
         {arriveTangent=6.3511, interpMode=RCIM_Cubic, leaveTangent=6.3511, tangentMode=RCTM_User, time=0, value=0}
         {interpMode=RCIM_Cubic, tangentMode=RCTM_User, time=0.5, value=1}
         {tangentMode=RCTM_User, time=1, value=0}
        ]}}})
    Scale RGB=&MakeVectorFromLinearColorRGB(Color=&FNTimeOfDayColor(
      Night Color=<rgba 0.1253 0.1914 0.5 1>
      Morning Color=<rgba 0.5 0.3383 0.3719 1>
      Day Color=<rgba 0.5 0.5 0.5 1>
      Evening Color=<rgba 0.4474 0.4547 0.4578 1>
     ))
    ScaleA=true
    ScaleRGB=true
   )
   /SetVariables(Particles.Position=&AddVectorToPosition(
     Position=$Particles.Initial.Position
     Vector=&Add_Vector(
      A=&MakeVector(
       X=0
       Y=0
       Z=&FloatFromCurve(FloatCurve=DataInterface() {propertyValues {Curve {keys [<at 0 -20>,<at 0.3 0>]}}}, CurveIndex=$Engine.System.Age)
      )
      B=&MakeVector(
       X=0
       Y=0
       Z=&Lerp_Float(
        A=&FloatFromCurve(FloatCurve=DataInterface() {propertyValues {Curve {keys [<at 0 -30>,<at 0.3 0>]}}}, CurveIndex=$Engine.System.Age)
        B=0
        Alpha=$Particles.SideDeterminism
       )
      )
     )
    )) {isSetParameters=true}
  ]
  rendererClasses [&NiagaraMeshRendererProperties]
  renderers [{rendererClass=&NiagaraMeshRendererProperties, rendererIndex=0}]
 }
}
summary {
 systemName=NS_QueenTruth_SkateboardBlur
 emitters [
  {emitterName=Main_Ring, rendererClasses [&NiagaraMeshRendererProperties]}
  {emitterName=BlurRing, rendererClasses [&NiagaraMeshRendererProperties]}
  {
   emitterName=Impact_Flare_and_Corona
   rendererClasses [&NiagaraSpriteRendererProperties, &NiagaraMeshRendererProperties]
  }
 ]
 userVariables [
  User.bIsFrontEnd {type=NiagaraBool}
  User.bIsFrontEndPreview {type=NiagaraBool}
  User.FrontEnd {type=NiagaraFloat, defaultValue=1, description=""}
  User.OwnerLocalVelocity {type=Vector3f, description="", defaultValue {struct=Vector3f, value=<0 0 0>}}
 ]
}
