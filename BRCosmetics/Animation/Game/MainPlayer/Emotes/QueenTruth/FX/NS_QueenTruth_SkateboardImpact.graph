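# Niagara impact effect for the skateboard in the QueenTruth player emote, pairing a short flash with a puff of dust-like smoke.
# Impact_Flare_and_Corona bursts two short-lived particles, offset apart by partition, drawn as fading sprites and an expanding mesh using dynamic material parameters.
# Smoke bursts sprites in a near-equatorial ring that push outward, slow under drag, drift upward, grow and fade.
# A holographic-squares emitter tinted by System.Color is present but disabled.
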NiagaraSystem /BRCosmetics/Animation/Game/MainPlayer/Emotes/QueenTruth/FX/NS_QueenTruth_SkateboardImpact
dependencies {
 usedDataInterfaces [
  &NiagaraDataInterfaceCurve
  &NiagaraDataInterfaceColorCurve
  &NiagaraDataInterfaceSkeletalMesh
  &NiagaraDataInterfaceArrayFloat
 ]
 usedDynamicInputs [&SelectFloatFromArray, &MakeIntFromFloat]
 usedModules [
  &SystemState
  &self:SystemSpawnScript.SystemScriptSource.SystemScriptGraph.NiagaraNodeAssignment_1.SetVariables_#1
  &self:SystemSpawnScript.SystemScriptSource.SystemScriptGraph.NiagaraNodeAssignment_2.SetVariables_#2
 ]
}
emitters {
 Em_Holo_Squares {
  bEnabled=false
  emitterUpdateScript [
   &EmitterState()
   &SpawnRate(Spawn Probability=0.5, SpawnRate=24, Use Spawn Probability=false)
  ]
  particleSpawnScript [
   &InitializeParticle(
    Color Mode=::NewEnumerator0
    Lifetime Max=1.5
    Lifetime Min=0.4
    Lifetime Mode=::NewEnumerator1
    Position Offset=<0 0 -25>
    Sprite Size Max=<6 8>
    Sprite Size Min=<2 2>
    Sprite Size Mode=::NewEnumerator2
    UsePositionOffset=true
    Write Material Random=true
   )
   &ShapeLocation(
    Cylinder Height=50
    Cylinder Radius=50
    MaskSpawnGroup=false
    Override Seed=false
    Shape Primitive=::NewEnumerator1
    Surface Only=false
    UseOwnerScale=true
   )
   &AddVelocity()
   &ApplyInitialForces()
  ]
  particleUpdateScript [
   &ParticleState()
   &ScaleSpriteSize(
    Uniform Curve Scale=0.75
    Uniform Curve Sprite Scale=DataInterface() {propertyValues {Curve {keys [{tangentMode=RCTM_User, time=0, value=0}, <at 0.25 1>, {tangentMode=RCTM_User, time=1, value=0}]}}}
   )
   &SolveForcesAndVelocity(Clamp Velocity=false, Limit Acceleration=false)
   &ScaleColor(
    Scale Alpha=&FloatFromCurve(FloatCurve=DataInterface() {propertyValues {Curve {keys [<at 0 0>,<at 0.3 1>,<at 1 0>]}}})
    Scale RGB=&Multiply_VectorByFloat(
     Vector=&MakeVectorFromLinearColorRGB(Color=$System.Color)
     Float=&FloatFromCurve(FloatCurve=DataInterface() {propertyValues {Curve {keys [<at 0 1>,<at 0.2 5>,<at 0.5 1>,<at 1 1>]}}})
    )
    ScaleA=true
    ScaleRGB=true
   )
   &ScaleColor(
    Scale Alpha=&Waveform("Waveform [1]"=::NewEnumerator0, "[1] Frequency"=0.3, "[1] Amplitude Min/Max"=<-1 1>)
    ScaleA=true
    ScaleRGB=false
   ) {enabled=false, module=ScaleColor001}
  ]
  rendererClasses [&NiagaraSpriteRendererProperties]
  renderers [{rendererClass=&NiagaraSpriteRendererProperties}]
 }
 Impact_Flare_and_Corona {
  emitterUpdateScript [
   &EmitterState()
   &SpawnBurst_Instantaneous(Spawn Count=2, Use Spawn Probability=false)
  ]
  particleSpawnScript [
   &PartitionParticles()
   /Niagara/Modules/Spawn/Initialization/InitializeParticle(
    Color=<rgba 1 1 1 0.3>
    Lifetime=0.2
    Mesh Scale=&UniformRangedVector(Minimum=<0.1 1.6 1.6>, Maximum=<0.1 1.8 1.8>)
    Module.Write Color=true
    Module.Write Lifetime=true
    Module.Write Mass=true
    Module.Write Material Random=true
    Module.Write Position=true
    Module.Write Scale=true
    Module.Write SpriteRotation=false
    Module.Write SpriteSize=true
    Position=&SimulationPosition()
    Sprite Size=<133.3333 116.6667>
   ) {enabled=false}
   /Niagara/Modules/Spawn/Initialization/V2/InitializeParticle(
    Color=<rgba 1 1 1 0.3>
    Lifetime=0.2
    Mesh Scale Max=<0.1 1.5 1.5>
    Mesh Scale Min=<0.1 1.3333 1.3333>
    Mesh Scale Mode=::NewEnumerator2
    Position Offset=&MaskVector(Mask Source=1, Mask Parameter=$Particles.PartitionParticles.Partition, Mask=0, Masked Vector=<0 -18 0>, Passthrough Vector=<0 22 0>)
    UsePositionOffset=true
    Write Material Random=true
   ) {module=InitializeParticle001}
   /SetVariables(Particles.CoronaSpriteSize=&Vector2DFromFloat(Value=120)) {isSetParameters=true}
   /SetVariables(Particles.ImpactColor:LinearColor=&MakeLinearColorFromVectorAndFloat("Vector (RGB)"=<0.5 0.5 0.5>, "Float (Alpha)"=&Multiply_Float(A=1, B=1))) {isSetParameters=true}
   &InitialMeshOrientation(Orientation Vector=<0 0 1>, Use Orientation Vector=true, Use Rotation Vector=true)
   &ApplyInitialForces()
   &AlignSpriteToMeshOrientation(SetAlignment=true, SetFacing=true)
  ]
  particleUpdateScript [
   &ParticleState()
   &DynamicMaterialParameters(
    Index 0 Param 1=&FloatFromCurve(FloatCurve=DataInterface() {propertyValues {Curve {keys [<at 0 1 RCIM_Cubic>,<at 0.8 0 RCIM_Cubic>]}}})
    Index 0 Param 2=2
    Index 0 Param 3=1
    Index 0 Param 4=1
    Index 1 Param 1=&FloatFromCurve(FloatCurve=DataInterface() {propertyValues {Curve {keys [
         {arriveTangent=-2.3163, interpMode=RCIM_Cubic, time=0, value=2}
         {arriveTangent=-1.9019, interpMode=RCIM_Cubic, leaveTangent=-1.9019, tangentMode=RCTM_User, time=1, value=0.4}
        ]}}})
    Index 1 Param 2=0.4
    Index 1 Param 3=0.6
    Index 1 Param 4=1
    Index 2 Param 1=&FloatFromCurve(FloatCurve=DataInterface() {propertyValues {Curve {keys [<at 0.46 0>,<at 1 1>]}}})
    Index 2 Param 2=45
    Module.Param0WriteEnabled=true
    Module.Param0WriteEnabled1=true
    Module.Param0WriteEnabled2=true
    Module.Param1WriteEnabled=true
    Module.Param1WriteEnabled1=true
    Module.Param1WriteEnabled2=true
    Module.Param2WriteEnabled=true
    Module.Param2WriteEnabled1=true
    Module.Param2WriteEnabled2=true
    Module.Param3WriteEnabled=true
    Module.Param3WriteEnabled1=true
    Module.Param3WriteEnabled2=true
    Write Parameter Index 1=true
    Write Parameter Index 2=true
   )
   /SetVariables(Particles.CoronaSpriteColor:LinearColor=&MakeLinearColorFromVectorAndFloat(
     "Vector (RGB)"=&MakeVectorFromLinearColorRGB(Color=<rgba 0.5521 0.5521 0.5521 1>)
     "Float (Alpha)"=&Multiply_Float(
      A=&FloatFromCurve(FloatCurve=DataInterface() {propertyValues {Curve {keys [<at 0.35 1.5 RCIM_Cubic>,<at 0.95 0>]}}})
      B=3.9
     )
    )) {isSetParameters=true}
   /SetVariables(Particles.ImpactColor:LinearColor=<rgba 0.4 0.4 0.4 1>) {isSetParameters=true}
   &ScaleColor(
    Scale Alpha=&FloatFromCurve(FloatCurve=DataInterface() {propertyValues {Curve {keys [<at 0 1>,<at 1 0>]}}})
    Scale RGB=&MakeVectorFromLinearColorRGB(Color=<rgba 1 1 1 1>)
    ScaleA=true
    ScaleRGB=true
   )
   &ScaleMeshSize(Scale Factor=&VectorFromFloat(Value=&FloatFromCurve(FloatCurve=DataInterface() {propertyValues {Curve {keys [<at 0 0.294 RCIM_Cubic>,<at 1 1.05>]}}}, Scale Curve=1.3)))
  ]
  rendererClasses [&NiagaraSpriteRendererProperties, &NiagaraMeshRendererProperties]
  renderers [
   {rendererClass=&NiagaraSpriteRendererProperties}
   {rendererClass=&NiagaraSpriteRendererProperties, rendererIndex=1}
   {rendererClass=&NiagaraMeshRendererProperties, rendererIndex=2}
  ]
 }
 Smoke {
  emitterUpdateScript [
   &EmitterState()
   &SpawnBurst_Instantaneous(
    Spawn Count=&RandomRangeInt(Minimum=15, Maximum=20)
    Use Spawn Probability=false
    UseLoopCountLimit=false
   )
  ]
  particleSpawnScript [
   &InitializeParticle(
    Color=<rgba 0.6376 0.5841 0.5333 0.1>
    Lifetime Max=1
    Lifetime Min=0.7
    Lifetime Mode=::NewEnumerator1
    Mesh Scale Mode=::NewEnumerator3
    Mesh Uniform Scale=150
    Sprite Rotation Mode=::NewEnumerator2
    Sprite Size Mode=::NewEnumerator1
    Uniform Sprite Size Max=3.9
    Uniform Sprite Size Min=3.25
    UsePositionOffset=true
    Write Material Random=true
   ) {module=InitializeParticle002}
   &ShapeLocation(
    MaskSpawnGroup=false
    Sphere Distribution=::NewEnumerator1
    Sphere Radius=&RandomRangeFloat(Minimum=10, Maximum=20)
    U Position=&ReturnNormalizedExecIndex()
    UseOwnerScale=true
    V Position=&RandomRangeFloat(Minimum=0.5, Maximum=0.55)
   )
   &AddVelocity(
    Velocity=&Multiply_VectorByFloat(Vector=$Particles.ShapeLocation.ShapeVector, Float=0.04)
    Velocity Speed Scale=&RandomRangeFloat(Minimum=250, Maximum=350)
   )
  ]
  particleUpdateScript [
   &ParticleState()
   &GravityForce(Gravity=<0 0 200>)
   &Drag(Drag=7, Use Linear Drag=true, Use Rotational Drag=false)
   &SolveForcesAndVelocity(Clamp Velocity=false, Limit Acceleration=false)
   &ScaleColor(
    Scale Alpha=&FloatFromCurve(FloatCurve=DataInterface() {propertyValues {Curve {keys [
         {arriveTangent=7.3326, interpMode=RCIM_Cubic, leaveTangent=7.3326, tangentMode=RCTM_User, time=0, value=0}
         {arriveTangent=0.0071, interpMode=RCIM_Cubic, leaveTangent=0.0071, tangentMode=RCTM_User, time=0.4, value=1}
         <at 1 0 RCIM_Cubic>
        ]}}})
    Scale RGB=&MakeVectorFromLinearColorRGB(Color=&ColorFromCurve(ColorCurve=DataInterface() {propertyValues {
        AlphaCurve {keys [<at 0 1 RCIM_Cubic>,<at 1 0 RCIM_Cubic>]}
        BlueCurve {keys [<at 0 1.2 RCIM_Cubic>,<at 0.4269 1 RCIM_Cubic>]}
        GreenCurve {keys [<at 0 1.2 RCIM_Cubic>,<at 0.4269 1 RCIM_Cubic>]}
        RedCurve {keys [<at 0 1.2 RCIM_Cubic>,<at 0.4269 1 RCIM_Cubic>]}
       }}))
    ScaleA=true
    ScaleRGB=true
   ) {module=ScaleColor001}
   &DynamicMaterialParameters(
    Index 0 Param 1=&Add_Float(
     A=&FloatFromCurve(FloatCurve=DataInterface() {propertyValues {Curve {keys [{arriveTangent=2.1549, time=0, value=0.8}, <at 1 1>]}}})
     B=&Multiply_Float(A=$Particles.Initial.MaterialRandom, B=0.5)
    )
    Index 1 Param 1=1
    Index 1 Param 2=1
    Index 1 Param 3=1
    Index 1 Param 4=1
    Index 2 Param 1=&FloatFromCurve(FloatCurve=DataInterface() {propertyValues {Curve {keys [
         {arriveTangent=1.8288, interpMode=RCIM_Cubic, leaveTangent=1.8288, tangentMode=RCTM_User, time=0, value=0.3}
         {arriveTangent=-0.0004, interpMode=RCIM_Cubic, leaveTangent=-0.0004, tangentMode=RCTM_User, time=1, value=1}
        ]}}})
    Index 2 Param 2=&RandomRangeFloat(Minimum=0.5, Maximum=1)
    Index 2 Param 3=4
    Index 2 Param 4=&RandomRangeFloat(Minimum=0.8, Maximum=1.2)
    Module.Param0WriteEnabled=true
    Module.Param0WriteEnabled1=true
    Module.Param0WriteEnabled2=true
    Module.Param0WriteEnabled3=true
    Module.Param1WriteEnabled=true
    Module.Param1WriteEnabled1=true
    Module.Param1WriteEnabled2=true
    Module.Param1WriteEnabled3=true
    Module.Param2WriteEnabled=true
    Module.Param2WriteEnabled1=true
    Module.Param2WriteEnabled2=true
    Module.Param2WriteEnabled3=true
    Module.Param3WriteEnabled=true
    Module.Param3WriteEnabled1=true
    Module.Param3WriteEnabled2=true
    Module.Param3WriteEnabled3=true
    Write Parameter Index 1=true
    Write Parameter Index 2=true
    Write Parameter Index 3=true
   ) {module=DynamicMaterialParameters002}
   &ScaleSpriteSize(Scale Factor=&Vector2DFromFloat(Value=&FloatFromCurve(FloatCurve=DataInterface() {propertyValues {Curve {keys [
          {arriveTangent=87.4549, interpMode=RCIM_Cubic, leaveTangent=87.4549, tangentMode=RCTM_User, time=0, value=1}
          {arriveTangent=4.8897, interpMode=RCIM_Cubic, leaveTangent=4.8897, tangentMode=RCTM_User, time=0.1, value=8}
          {interpMode=RCIM_Cubic, leaveTangent=-6.7553, time=1.0011, value=10}
         ]}}})))
   &CameraOffset(Camera Offset Amount=50)
   &SpriteRotationRate(Rotation Rate=&RandomRangeFloat(Minimum=-90, Maximum=90))
   &SpriteRotationRate(Rotation Rate=&Multiply_Float(A=90, B=&RandomRangeFloat(Minimum=-1, Maximum=1))) {module=SpriteRotationRate001}
  ]
  rendererClasses [&NiagaraSpriteRendererProperties]
  renderers [
   {rendererClass=&NiagaraSpriteRendererProperties}
   {rendererClass=&NiagaraSpriteRendererProperties, rendererIndex=1}
  ]
 }
}
summary {
 systemName=NS_QueenTruth_SkateboardImpact
 emitters [
  {bEnabled=false, emitterName=Em_Holo_Squares, rendererClasses [&NiagaraSpriteRendererProperties]}
  {emitterName=Smoke, rendererClasses [&NiagaraSpriteRendererProperties]}
  {
   emitterName=Impact_Flare_and_Corona
   rendererClasses [&NiagaraSpriteRendererProperties, &NiagaraMeshRendererProperties]
  }
 ]
 userVariables [
  User.bIsFrontEnd {type=NiagaraBool}
  User.bIsFrontEndPreview {type=NiagaraBool}
  User.HologramColorLerp {type=NiagaraFloat, defaultValue=0, description=""}
  User.Is WeakPoint {type=NiagaraBool, defaultValue=true}
  User.WorkoutSelect {type=NiagaraFloat, defaultValue=0, description=""}
 ]
}
